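# Animation Blueprint for the SquibblySquid pickaxe, adding secondary physics motion to its tongue, hook, spine and head bones.
# The input pose runs through chained bone-relative Anim Dynamics nodes, then a per-bone blend layers in DefaultSlot, weighted by the front-end Frontend_Idle_Bypass curve.
# Each update sets gravity from a parameter asset and picks alpha, spring and sim-space variables for in-zone, front-end on-pawn or front-end off-pawn states.
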AnimBlueprint /BRCosmetics/Weapons/FORT_Melee/Pickaxe_SquibblySquid/Meshes/SquibblySquid_Axe_AnimBP
target_skeleton=&/SquibblySquid_Axe_Skeleton
graphs [
 AnimGraph {nodes [
   AnimGraphNode_SubInput_0 = Input Pose()
   n1 = Anim Dynamics - Bone: dyn_tongue_00!(
    ComponentPose=Local To Component(LocalPose=@AnimGraphNode_SubInput_0)
    Alpha=1
    GravityOverride:Vector="0,0,-980"
    AngularSpringConstant:float=300
    SimSpaceSettings:Anim Phys Sim Space Settings Structure=(SimSpaceAngularAlpha=0,MaxAngularVelocity=10000,MaxAngularAcceleration=10000,ExternalAngularVelocity=(X=0,Y=0,Z=0))
    ExternalForce="0,0,0"
   ) -> (Pose) {
    angularDampingOverride=1
    angularSpringConstant=300
    bAngularSpring=true
    bOverrideAngularDamping=true
    bUseGravityOverride=true
    gravityScale=0.35
    simulationSpace=BoneRelative
    boundBone {boneName=dyn_tongue_00}
    componentAppliedLinearAccClamp=<3000 3000 3000>
    componentLinearAccScale=<0.8 0.8 0.8>
    componentLinearVelScale=<0.8 0.8 0.8>
    gravityOverride {z=-980}
    physicsBodyDefinitions [{
      boundBone {boneName=dyn_tongue_00}
      boxExtents=<10 10 10>
      constraintSetup {angularTargetAxis=AxisY, angularLimitsMax=<0 -10 20>, angularLimitsMin=<0 -20 -20>, angularTarget=<1 0 0>}
      localJointOffset=<0 4 0>
     }]
    relativeSpaceBone {boneName=b_AR_Root}
   }
   n2 = Anim Dynamics - Bone: dyn_tongue_01!(
    ComponentPose=@n1
    Alpha=1
    GravityOverride:Vector="0,0,-980"
    AngularSpringConstant:float=300
    SimSpaceSettings:Anim Phys Sim Space Settings Structure=(SimSpaceAngularAlpha=0,MaxAngularVelocity=10000,MaxAngularAcceleration=10000,ExternalAngularVelocity=(X=0,Y=0,Z=0))
    ExternalForce="0,0,0"
   ) -> (Pose) {
    angularDampingOverride=1
    angularSpringConstant=300
    bAngularSpring=true
    bOverrideAngularDamping=true
    bUseGravityOverride=true
    gravityScale=0.35
    simulationSpace=BoneRelative
    boundBone {boneName=dyn_tongue_01}
    componentAppliedLinearAccClamp=<3000 3000 3000>
    componentLinearAccScale=<0.8 0.8 0.8>
    componentLinearVelScale=<0.8 0.8 0.8>
    gravityOverride {z=-980}
    physicsBodyDefinitions [{
      boundBone {boneName=dyn_tongue_01}
      boxExtents=<10 10 10>
      constraintSetup {angularTargetAxis=AxisY, angularLimitsMax=<0 30 30>, angularLimitsMin=<0 -30 -30>, angularTarget=<1 0 0>}
      localJointOffset=<4 4 0>
     }]
    relativeSpaceBone {boneName=b_AR_Root}
   }
   n3 = Anim Dynamics - Bone: dyn_hook_00!(
    ComponentPose=@n2
    Alpha=1
    GravityOverride:Vector="0,0,-980"
    AngularSpringConstant:float=300
    SimSpaceSettings:Anim Phys Sim Space Settings Structure=(SimSpaceAngularAlpha=0,MaxAngularVelocity=10000,MaxAngularAcceleration=10000,ExternalAngularVelocity=(X=0,Y=0,Z=0))
    ExternalForce="0,0,0"
   ) -> (Pose) {
    angularDampingOverride=1
    angularSpringConstant=300
    bAngularSpring=true
    bOverrideAngularDamping=true
    bUseGravityOverride=true
    gravityScale=0.35
    simulationSpace=BoneRelative
    boundBone {boneName=dyn_hook_00}
    componentAppliedLinearAccClamp=<3000 3000 3000>
    componentLinearAccScale=<0.8 0.8 0.8>
    componentLinearVelScale=<0.8 0.8 0.8>
    gravityOverride {z=-980}
    physicsBodyDefinitions [{
      boundBone {boneName=dyn_hook_00}
      boxExtents=<10 10 10>
      constraintSetup {angularTargetAxis=AxisY, angularLimitsMax=<15 -5 15>, angularLimitsMin=<-15 -25 -10>, angularTarget=<1 0 0>}
      localJointOffset=<1 0 0>
     }]
    relativeSpaceBone {boneName=b_AR_Root}
   }
   n4 = Anim Dynamics - Bone: dyn_spine_00!(
    ComponentPose=@n3
    Alpha=1
    GravityOverride:Vector="0,0,-980"
    AngularSpringConstant:float=300
    SimSpaceSettings:Anim Phys Sim Space Settings Structure=(SimSpaceAngularAlpha=1,MaxAngularVelocity=7,MaxAngularAcceleration=35,ExternalAngularVelocity=(X=0,Y=0,Z=0))
    ExternalForce="0,0,0"
   ) -> (Pose) {
    angularDampingOverride=0.9813
    angularSpringConstant=300
    bAngularSpring=true
    bOverrideAngularDamping=true
    bUseGravityOverride=true
    gravityScale=0.35
    simulationSpace=BoneRelative
    boundBone {boneName=dyn_spine_00}
    componentAppliedLinearAccClamp=<3000 3000 3000>
    componentLinearAccScale=<0.2 0.2 0.2>
    componentLinearVelScale=<0.2 0.2 0.2>
    gravityOverride {z=-980}
    physicsBodyDefinitions [{
      boundBone {boneName=dyn_spine_00}
      boxExtents=<10 10 10>
      constraintSetup {angularTargetAxis=AxisY, angularLimitsMax=<10 10 5>, angularLimitsMin=<-10 0 -20>, angularTarget=<1 0 0>}
      localJointOffset=<1 0 0>
     }]
    relativeSpaceBone {boneName=b_AR_Root}
    simSpaceSettings {maxAngularAcceleration=35, maxAngularVelocity=7, simSpaceAngularAlpha=1}
   }
   n5 = Anim Dynamics - Bone: dyn_spine_01!(
    ComponentPose=@n4
    Alpha=1
    GravityOverride:Vector="0,0,-980"
    AngularSpringConstant:float=300
    SimSpaceSettings:Anim Phys Sim Space Settings Structure=(SimSpaceAngularAlpha=1,MaxAngularVelocity=7,MaxAngularAcceleration=35,ExternalAngularVelocity=(X=0,Y=0,Z=0))
    ExternalForce="0,0,0"
   ) -> (Pose) {
    angularDampingOverride=0.9813
    angularSpringConstant=300
    bAngularSpring=true
    bOverrideAngularDamping=true
    bUseGravityOverride=true
    gravityScale=0.35
    simulationSpace=BoneRelative
    boundBone {boneName=dyn_spine_01}
    componentAppliedLinearAccClamp=<3000 3000 3000>
    componentLinearAccScale=<0.2 0.2 0.2>
    componentLinearVelScale=<0.2 0.2 0.2>
    gravityOverride {z=-980}
    physicsBodyDefinitions [{
      boundBone {boneName=dyn_spine_01}
      boxExtents=<10 10 10>
      constraintSetup {angularTargetAxis=AxisY, angularLimitsMax=<10 10 5>, angularLimitsMin=<-10 0 -10>, angularTarget=<1 0 0>}
      localJointOffset=<1 0 0>
     }]
    relativeSpaceBone {boneName=b_AR_Root}
    simSpaceSettings {maxAngularAcceleration=35, maxAngularVelocity=7, simSpaceAngularAlpha=1}
   }
   n6 = Anim Dynamics - Bone: dyn_spine_02!(
    ComponentPose=@n5
    Alpha=1
    GravityOverride:Vector="0,0,-980"
    AngularSpringConstant:float=300
    SimSpaceSettings:Anim Phys Sim Space Settings Structure=(SimSpaceAngularAlpha=0,MaxAngularVelocity=10000,MaxAngularAcceleration=10000,ExternalAngularVelocity=(X=0,Y=0,Z=0))
    ExternalForce="0,0,0"
   ) -> (Pose) {
    angularDampingOverride=0.9813
    angularSpringConstant=300
    bAngularSpring=true
    bOverrideAngularDamping=true
    bUseGravityOverride=true
    gravityScale=0.35
    simulationSpace=BoneRelative
    boundBone {boneName=dyn_spine_02}
    componentAppliedLinearAccClamp=<3000 3000 3000>
    componentLinearAccScale=<0.15 0.15 0.15>
    componentLinearVelScale=<0.15 0.15 0.15>
    gravityOverride {z=-980}
    physicsBodyDefinitions [{
      boundBone {boneName=dyn_spine_02}
      boxExtents=<10 10 10>
      constraintSetup {angularTargetAxis=AxisY, angularLimitsMax=<10 10 5>, angularLimitsMin=<-10 0 -5>, angularTarget=<1 0 0>}
      localJointOffset=<1 0 0>
     }]
    relativeSpaceBone {boneName=b_AR_Root}
    simSpaceSettings {maxAngularAcceleration=35, maxAngularVelocity=7, simSpaceAngularAlpha=1}
   }
   n7 = Anim Dynamics - Bone: dyn_spine_03!(
    ComponentPose=@n6
    Alpha=1
    GravityOverride:Vector="0,0,-980"
    AngularSpringConstant:float=300
    SimSpaceSettings:Anim Phys Sim Space Settings Structure=(SimSpaceAngularAlpha=1,MaxAngularVelocity=7,MaxAngularAcceleration=35,ExternalAngularVelocity=(X=0,Y=0,Z=0))
    ExternalForce="0,0,0"
   ) -> (Pose) {
    angularDampingOverride=0.9813
    angularSpringConstant=300
    bAngularSpring=true
    bOverrideAngularDamping=true
    bUseGravityOverride=true
    gravityScale=0.35
    simulationSpace=BoneRelative
    boundBone {boneName=dyn_spine_03}
    componentAppliedLinearAccClamp=<3000 3000 3000>
    componentLinearAccScale=<0.15 0.15 0.15>
    componentLinearVelScale=<0.15 0.15 0.15>
    gravityOverride {z=-980}
    physicsBodyDefinitions [{
      boundBone {boneName=dyn_spine_03}
      boxExtents=<10 10 10>
      constraintSetup {angularTargetAxis=AxisY, angularLimitsMax=<10 10 5>, angularLimitsMin=<-10 0 -10>, angularTarget=<1 0 0>}
      localJointOffset=<1 0 0>
     }]
    relativeSpaceBone {boneName=b_AR_Root}
    simSpaceSettings {maxAngularAcceleration=35, maxAngularVelocity=7, simSpaceAngularAlpha=1}
   }
   n8 = Anim Dynamics - Bone: dyn_head_00!(
    ComponentPose=@n7
    Alpha=1
    GravityOverride:Vector="0,0,-980"
    AngularSpringConstant:float=300
    SimSpaceSettings:Anim Phys Sim Space Settings Structure=(SimSpaceAngularAlpha=1,MaxAngularVelocity=7,MaxAngularAcceleration=35,ExternalAngularVelocity=(X=0,Y=0,Z=0))
    ExternalForce="0,0,0"
   ) -> (Pose) {
    angularDampingOverride=0.9813
    angularSpringConstant=300
    bAngularSpring=true
    bOverrideAngularDamping=true
    bUseGravityOverride=true
    gravityScale=0.35
    simulationSpace=BoneRelative
    boundBone {boneName=dyn_head_00}
    componentAppliedLinearAccClamp=<3000 3000 3000>
    componentLinearAccScale=<0.1 0.1 0.1>
    componentLinearVelScale=<0.1 0.1 0.1>
    gravityOverride {z=-980}
    physicsBodyDefinitions [{
      boundBone {boneName=dyn_head_00}
      boxExtents=<10 10 10>
      constraintSetup {angularTargetAxis=AxisY, angularLimitsMax=<10 10 5>, angularLimitsMin=<0 0 -15>, angularTarget=<1 0 0>}
      localJointOffset=<1 0 0>
     }]
    relativeSpaceBone {boneName=b_AR_Root}
    simSpaceSettings {maxAngularAcceleration=35, maxAngularVelocity=7, simSpaceAngularAlpha=1}
   }
   n9 = Layered blend per bone(
    BasePose=Component To Local(ComponentPose=@n8)
    BlendPoses_0=Slot 'DefaultSlot'()
    BlendWeights_0=Get Custom Idle Is Playing() -> (Custom Idle Is Playing:double)
   ) {layerSetup [{branchFilters [
       {boneName=b_AR_Root}
       {boneName=Melee_TwoHanded_Handle}
       {boneName=dyn_spine_MSTR}
       {boneName=dyn_spine_00}
       {boneName=dyn_spine_01}
       {boneName=dyn_spine_02}
       {boneName=dyn_spine_03}
       {boneName=dyn_head_MSTR}
       {boneName=dyn_head_00}
       {boneName=dyn_hook_MSTR}
       {boneName=dyn_hook_00}
       {boneName=dyn_tongue_MSTR}
       {boneName=dyn_tongue_00}
       {boneName=dyn_tongue_01}
      ]}]}
   Output Pose(Result=@n9) {layerGroup=None}
  ]}
 EventGraph {nodes [
   Event BlueprintUpdateAnimation()
   > n1 = Sequence!(execute) -> (then_0, then_1, then_2:Exec)
   > n2 = "Gravity Override State Function (Parameter Asset) (Nativized)"(
    CharacterPartAnimInstance=Self-Reference()
    GameTimeInSeconds=0
    Parameters=&/Fortnite_SquibblySquid_Axe_Parameters
   )
   > Set Gravity Override(Gravity Override=@n2)
   n3 = Branch(execute=@n1.then_1, Condition=Get bIsInFrontEnd())
   > n4 = Branch(Condition=Get bIsOnPawn() -> (bIsOnPawn:bool))
   > Set Alpha FE(Alpha FE:float=0.65) -> (Output_Get:float)
   > Set Angular Spring Constant(Angular Spring Constant:float=300) -> (Output_Get:float)
   > Set Tongue Alpha FE(Tongue Alpha FE:float=0.5) -> (Output_Get:float)
   > Set Tongue Angular Spring Constant(Tongue Angular Spring Constant:float=300) -> (Output_Get:float)
   > Set Sim Space Settings(Sim Space Settings:Anim Phys Sim Space Settings Structure=Get Sim Space Settings Front End() -> (Sim Space Settings Front End:Anim Phys Sim Space Settings Structure)) -> (Output_Get:Anim Phys Sim Space Settings Structure)
   Set Alpha FE(execute=@n4.else, Alpha FE=0.5)
   > Set Angular Spring Constant(Angular Spring Constant=75)
   > Set Tongue Alpha FE(Tongue Alpha FE=0.8)
   > Set Tongue Angular Spring Constant(Tongue Angular Spring Constant=125)
   > Set Sim Space Settings(Sim Space Settings=Get Sim Space Settings Front End())
   Set Alpha FE(execute=@n3.else, Alpha FE=1)
   > Set Angular Spring Constant(Angular Spring Constant=300)
   > Set Tongue Alpha FE(Tongue Alpha FE=1)
   > Set Tongue Angular Spring Constant(Tongue Angular Spring Constant=300)
   > Set Sim Space Settings(Sim Space Settings=Get Sim Space Settings in Zone() -> (Sim Space Settings in Zone:Anim Phys Sim Space Settings Structure))
   Branch(execute=@n1.then_2, Condition=Get bIsInFrontEnd())
   > Set Custom Idle Is Playing(Custom Idle Is Playing:double=GetCurveValue(self=GetOwnerFrontEndAnimInstance(), CurveName=Frontend_Idle_Bypass)) -> (Output_Get:double)
   TryGetPawnOwner()
  ]}
]
variables [
 Alpha FE {type=number, default=1}
 Angular Spring Constant {type=number, default=300}
 Custom Idle Is Playing {type=number}
 Gravity Override {type=Vector}
 Sim Space Settings {
  type=AnimPhysSimSpaceSettings
  default {maxAngularAcceleration=35, maxAngularVelocity=7, simSpaceAngularAlpha=1, externalAngularVelocity=<0 0 0>}
 }
 Sim Space Settings Front End {
  type=AnimPhysSimSpaceSettings
  default {maxAngularAcceleration=15, maxAngularVelocity=7, simSpaceAngularAlpha=0.7, externalAngularVelocity=<0 0 0>}
 }
 Sim Space Settings in Zone {
  type=AnimPhysSimSpaceSettings
  default {maxAngularAcceleration=35, maxAngularVelocity=7, simSpaceAngularAlpha=1, externalAngularVelocity=<0 0 0>}
 }
 Tongue Alpha FE {type=number, default=1}
 Tongue Angular Spring Constant {type=number, default=300}
]
